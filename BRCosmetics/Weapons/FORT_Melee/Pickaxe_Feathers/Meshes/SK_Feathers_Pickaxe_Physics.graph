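# Physics asset for the Feathers pickaxe skeletal mesh, a two-handed melee cosmetic, giving its five dyn_whisk bones slight physical motion.
# A long, thin capsule covers the handle bone, and each whisk bone has its own wider capsule.
# Each whisk is constrained directly to the handle rather than chained, with very tight swing limits and somewhat looser, per-bone twist limits, so the pieces only wobble subtly.
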PhysicsAsset /BRCosmetics/Weapons/FORT_Melee/Pickaxe_Feathers/Meshes/SK_Feathers_Pickaxe_Physics
bodies {
 Melee_TwoHanded_Handle [{shapeType=Capsule, length=55.148, radius=3.4827, center=<0.3697 -1.0803 14.4283>, rotation=<rot -0.0413 -90 -0.0609>}]
 dyn_whisk_01 [{shapeType=Capsule, length=19.0051, radius=8.3189, center=<-1.6028 -0.3091 22.7449>}]
 dyn_whisk_02 [{shapeType=Capsule, length=20.4591, radius=7.4997, center=<0.1505 0.1975 19.1036>}]
 dyn_whisk_03 [{shapeType=Capsule, length=24.4002, radius=7.603, center=<-0.7894 0.5855 18.5542>}]
 dyn_whisk_04 [{shapeType=Capsule, length=22.3643, radius=7.9123, center=<1.1726 0.4633 22.0413>}]
 dyn_whisk_05 [{shapeType=Capsule, length=25.6426, radius=7.7571, center=<1.0272 0.0923 13.1264>}]
}
constraints [
 {bone1Name=dyn_whisk_01, bone2Name=Melee_TwoHanded_Handle, swing1LimitDegrees=1, swing2LimitDegrees=1, twistLimitDegrees=8}
 {bone1Name=dyn_whisk_02, bone2Name=Melee_TwoHanded_Handle, swing1LimitDegrees=2, swing2LimitDegrees=2, twistLimitDegrees=7}
 {bone1Name=dyn_whisk_03, bone2Name=Melee_TwoHanded_Handle, swing1LimitDegrees=1, swing2LimitDegrees=1, twistLimitDegrees=4}
 {bone1Name=dyn_whisk_04, bone2Name=Melee_TwoHanded_Handle, swing1LimitDegrees=2, swing2LimitDegrees=2, twistLimitDegrees=10}
 {bone1Name=dyn_whisk_05, bone2Name=Melee_TwoHanded_Handle, swing1LimitDegrees=1, swing2LimitDegrees=1, twistLimitDegrees=2}
]
